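# Sound cue for exiting a portapotty hideable, used by gameplay cue audio.
# A Random node, excluded from branch culling, picks between two exit wave variations with equal weight.
# A Modulator applies a fixed volume boost on top, offset by a low cue volume multiplier.
# It plays with mono trap attenuation under the gadgets and items sound class.
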SoundCue /FNGameplayCues/Sounds/Fort_Hideables/Portapotty/Hideable_Portapotty_Exit_Cue
root=Modulator_0
nodes [
 {id=Modulator_0, children [Random_0], properties {volumeMax=2.5, volumeMin=2.5}}
 {id=Random_0, children [WavePlayer_0,WavePlayer_3], properties {bShouldExcludeFromBranchCulling=true, weights [1,1]}}
 {id=WavePlayer_0, properties {soundWaveAssetPtr=&/Hideable_Portapotty_Exit_01}}
 {id=WavePlayer_3, properties {soundWaveAssetPtr=&/Hideable_Portapotty_Exit_02}}
]
properties {
 attenuationSettings=/Game/Sounds/Attenuation/Fort_Traps_Mono_Att
 maxDistance=6000
 soundClassObject=&Fort_Gadgets_Items
 volumeMultiplier=0.2857
}
