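# Sound cue for the rustle of the portapotty hideable prop, with a rare fart variant as a gag.
# A top random node mostly picks one of four rustle waves and occasionally falls into a low-weight nested random that chooses among three farts and a fifth rustle.
# A modulator at the root varies volume, and the cue uses the hiding-prop rustle attenuation and the gadgets/items sound class.
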SoundCue /FNGameplayCues/Sounds/Fort_Hideables/Portapotty/Hideable_Portapotty_Rustle_Cue
root=Modulator_0
nodes [
 {id=Modulator_0, children [Random_3], properties {volumeMax=1.5, volumeMin=1.3}}
 {
  id=Random_2
  children [WavePlayer_0,WavePlayer_1,WavePlayer_7,WavePlayer_2]
  properties {bShouldExcludeFromBranchCulling=true, weights [0.5,0.5,0.5,1]}
 }
 {
  id=Random_3
  children [WavePlayer_3,WavePlayer_4,WavePlayer_5,WavePlayer_6,Random_2]
  properties {bShouldExcludeFromBranchCulling=true, weights [1,1,1,1,0.1]}
 }
 {id=WavePlayer_0, properties {soundWaveAssetPtr=&/Hideable_Portapotty_Rustle_Fart_01}}
 {id=WavePlayer_1, properties {soundWaveAssetPtr=&/Hideable_Portapotty_Rustle_Fart_02}}
 {id=WavePlayer_2, properties {soundWaveAssetPtr=&/Hideable_Portapotty_Rustle_05}}
 {id=WavePlayer_3, properties {soundWaveAssetPtr=&/Hideable_Portapotty_Rustle_01}}
 {id=WavePlayer_4, properties {soundWaveAssetPtr=&/Hideable_Portapotty_Rustle_02}}
 {id=WavePlayer_5, properties {soundWaveAssetPtr=&/Hideable_Portapotty_Rustle_03}}
 {id=WavePlayer_6, properties {soundWaveAssetPtr=&/Hideable_Portapotty_Rustle_04}}
 {id=WavePlayer_7, properties {soundWaveAssetPtr=&/Hideable_Portapotty_Rustle_Fart_03}}
]
properties {
 attenuationSettings=/FNGameplayCues/Sounds/Attenuation/Fort_HidingProps_Rustle
 maxDistance=3000
 soundClassObject=&Fort_Gadgets_Items
 volumeMultiplier=0.6857
}
